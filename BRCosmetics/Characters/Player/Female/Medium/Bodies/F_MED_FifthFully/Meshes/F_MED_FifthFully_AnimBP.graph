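# Post-process style AnimBlueprint for the F_MED_FifthFully female body cosmetic, following the attached parent mesh and adding secondary motion.
# It copies the parent pose, drives the right bag base bone with a weighted constraint to pelvis, thigh and deform bones, then runs a Control Rig.
# The cached result feeds chained root-relative AnimDynamics nodes for the right boot tongue, bags and front jacket flaps.
# The event graph holds no logic.
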AnimBlueprint /BRCosmetics/Characters/Player/Female/Medium/Bodies/F_MED_FifthFully/Meshes/F_MED_FifthFully_AnimBP
target_skeleton=&/F_MED_FifthFully_Skeleton
graphs [
 AnimGraph {nodes [
   n1 = Anim Dynamics - Bone: dyn_boot_tongue_r_01(ComponentPose=Local To Component(LocalPose=Use cached pose 'to AnimDyn'()), Alpha=1, ExternalForce="0,0,0") {
    angularSpringConstant=1000
    bAngularSpring=true
    bOverrideLinearDamping=true
    bUseGravityOverride=true
    linearDampingOverride=1
    simulationSpace=RootRelative
    boundBone {boneName=dyn_boot_tongue_r_01}
    componentAppliedLinearAccClamp=<5000 5000 5000>
    componentLinearAccScale=<1 1 1>
    componentLinearVelScale=<1 1 1>
    gravityOverride {z=-980}
    physicsBodyDefinitions [{
      boundBone {boneName=dyn_boot_tongue_r_01}
      boxExtents=<10 10 10>
      constraintSetup {angularLimitsMax=<0 0 10>, angularLimitsMin=<0 0 -10>, angularTarget=<1 0 0>}
      localJointOffset=<1 0 0>
     }]
   }
   n2 = Anim Dynamics - Bone: dyn_bag_l_01(ComponentPose=@n1, Alpha=1, ExternalForce="0,0,0") {
    bOverrideLinearDamping=true
    bUseGravityOverride=true
    linearDampingOverride=1
    simulationSpace=RootRelative
    boundBone {boneName=dyn_bag_l_01}
    componentAppliedLinearAccClamp=<5000 5000 5000>
    componentLinearAccScale=<0.6 0.4 0.4>
    componentLinearVelScale=<0.6 0.4 0.4>
    gravityOverride {z=-980}
    physicsBodyDefinitions [{
      boundBone {boneName=dyn_bag_l_01}
      boxExtents=<10 10 10>
      constraintSetup {angularLimitsMax=<0 0 0>, angularLimitsMin=<0 -22 0>, angularTarget=<1 0 0>}
      localJointOffset=<0.5 0 0>
     }]
   }
   n3 = Anim Dynamics - Bone: dyn_bag_r_01(ComponentPose=@n2, Alpha=1, ExternalForce="0,0,0") {
    bOverrideLinearDamping=true
    bUseGravityOverride=true
    linearDampingOverride=1
    simulationSpace=RootRelative
    boundBone {boneName=dyn_bag_r_01}
    componentAppliedLinearAccClamp=<8000 8000 8000>
    componentLinearAccScale=<0.5 0.2 0.2>
    componentLinearVelScale=<0.5 0.2 0.2>
    gravityOverride {z=-980}
    physicsBodyDefinitions [{
      boundBone {boneName=dyn_bag_r_01}
      boxExtents=<10 10 10>
      constraintSetup {angularLimitsMax=<0 0 0>, angularLimitsMin=<0 -5 0>, angularTarget=<1 0 0>}
      localJointOffset=<0.1 0 0>
     }]
   }
   n4 = Anim Dynamics - Bone: dyn_bag_r_02(ComponentPose=@n3, Alpha=1, ExternalForce="0,0,0") {
    angularSpringConstant=800
    bOverrideLinearDamping=true
    bUseGravityOverride=true
    bUsePlanarLimit=false
    linearDampingOverride=1
    simulationSpace=RootRelative
    boundBone {boneName=dyn_bag_r_02}
    componentAppliedLinearAccClamp=<2500 2500 2500>
    componentLinearAccScale=<0.6 0.5 0.5>
    componentLinearVelScale=<0.6 0.5 0.5>
    gravityOverride {z=-980}
    physicsBodyDefinitions [{
      sphereCollisionRadius=8
      boundBone {boneName=dyn_bag_r_02}
      boxExtents=<5 5 5>
      constraintSetup {angularLimitsMax=<0 0 0>, angularLimitsMin=<0 -5 -10>, angularTarget=<1 0 0>}
      localJointOffset=<0.1 0 0>
     }]
    sphericalLimits [{limitRadius=18.2477, drivingBone {boneName=thigh_twist_01_r}, sphereLocalOffset=<3.4834 13.6063 -11.2222>}]
   }
   n5 = Anim Dynamics - Bone: dyn_bag_r_03(ComponentPose=@n4, Alpha=1, ExternalForce="0,0,0") {
    angularSpringConstant=800
    bOverrideLinearDamping=true
    bUseGravityOverride=true
    bUsePlanarLimit=false
    linearDampingOverride=1
    simulationSpace=RootRelative
    boundBone {boneName=dyn_bag_r_03}
    componentAppliedLinearAccClamp=<2500 2500 2500>
    componentLinearAccScale=<0.6 0.5 0.5>
    componentLinearVelScale=<0.6 0.5 0.5>
    gravityOverride {z=-980}
    physicsBodyDefinitions [{
      sphereCollisionRadius=8
      boundBone {boneName=dyn_bag_r_03}
      boxExtents=<5 5 5>
      constraintSetup {angularLimitsMax=<0 0 0>, angularLimitsMin=<0 -5 -10>, angularTarget=<1 0 0>}
      localJointOffset=<0.1 0 0>
     }]
    sphericalLimits [{limitRadius=18.2477, drivingBone {boneName=thigh_twist_01_r}, sphereLocalOffset=<3.4834 13.6063 -11.2222>}]
   }
   n6 = Anim Dynamics - Bone: dyn_jacket_frnt_l_01(ComponentPose=@n5, Alpha=1, ExternalForce="0,0,0") {
    bOverrideLinearDamping=true
    bUseGravityOverride=true
    bUseSphericalLimits=true
    linearDampingOverride=1
    simulationSpace=RootRelative
    boundBone {boneName=dyn_jacket_frnt_l_01}
    componentAppliedLinearAccClamp=<5000 5000 5000>
    componentLinearAccScale=<0.5 0.5 0.5>
    componentLinearVelScale=<0.5 0.5 0.5>
    gravityOverride {z=-980}
    physicsBodyDefinitions [{
      sphereCollisionRadius=8
      boundBone {boneName=dyn_jacket_frnt_l_01}
      boxExtents=<10 10 10>
      constraintSetup {angularLimitsMax=<0 5 5>, angularLimitsMin=<0 -5 -5>, angularTarget=<1 0 0>}
      localJointOffset=<0.3 0 0>
     }]
   }
   n7 = Anim Dynamics - Bone: dyn_jacket_frnt_r_01(ComponentPose=@n6, Alpha=1, ExternalForce="0,0,0") {
    bOverrideLinearDamping=true
    bUseGravityOverride=true
    bUseSphericalLimits=true
    linearDampingOverride=1
    simulationSpace=RootRelative
    boundBone {boneName=dyn_jacket_frnt_r_01}
    componentAppliedLinearAccClamp=<5000 5000 5000>
    componentLinearAccScale=<0.5 0.5 0.5>
    componentLinearVelScale=<0.5 0.5 0.5>
    gravityOverride {z=-980}
    physicsBodyDefinitions [{
      sphereCollisionRadius=8
      boundBone {boneName=dyn_jacket_frnt_r_01}
      boxExtents=<10 10 10>
      constraintSetup {angularLimitsMax=<0 5 5>, angularLimitsMin=<0 -5 -5>, angularTarget=<1 0 0>}
      localJointOffset=<0.3 0 0>
     }]
   }
   Output Pose(Result=Component To Local(ComponentPose=@n7)) {layerGroup=None}
   n8 = "Constraint - dyn_bag_r_base (PPPPP)"!(
    ComponentPose=Local To Component(LocalPose=Copy Pose From Mesh() {bUseAttachedParent=true})
    Alpha=1
    ConstraintWeights_0=0.0162
    ConstraintWeights_1:float=0.4619
    ConstraintWeights_2:float=0.521
    ConstraintWeights_3:float=0
    ConstraintWeights_4:float=0.0009
   ) -> (Pose) {
    boneToModify {boneName=dyn_bag_r_base}
    constraintSetup [
     {targetBone {boneName=pelvis}}
     {targetBone {boneName=thigh_twist_01_r}}
     {targetBone {boneName=thigh_r}}
     {targetBone {boneName=deform_glute_r}}
     {targetBone {boneName=deform_groin_r}}
    ]
    constraintWeights [1,1,1,1,1]
   }
   n9 = Control Rig(Source=Component To Local(ComponentPose=@n8)) {controlRigAssetReference {controlRigAsset=&/F_MED_FifthFully_CtrlRig}}
   Save cached pose 'to AnimDyn'(Pose=@n9) {cacheName=to AnimDyn}
  ]}
 EventGraph {nodes [Event BlueprintUpdateAnimation()]}
]
variables []
