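# Physics asset for the FifthFully female character's face accessory mesh, set up for dynamic long hair with upper-body collision shapes.
# Capsules cover pelvis, spine, neck, head, clavicles and upper arms, with added shoulder spheres, all linked by free angular constraints.
# Six branching hair chains (three back, two back-side, right front) hang from zero-size base bodies on the head.
# Per-joint angle limits keep back chains mostly to one swing axis, while side and front chains generally bend on both.
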PhysicsAsset /BRCosmetics/Characters/Player/Female/Medium/Bodies/F_MED_FifthFully/Meshes/Parts/F_MED_FifthFully_FaceAcc_Physics_IK
bodies {
 clavicle_l [{shapeType=Capsule, length=8.6716, radius=4.4661, shapeName=clavicle_l_capsule, center=<5.2488 -1.4201 0>, rotation=<rot 90 0 0>}]
 clavicle_r [{shapeType=Capsule, length=10.4097, radius=5.8216, shapeName=clavicle_r_capsule, center=<-5.2488 1.4201 0>, rotation=<rot 90 180 0>}]
 dyn_hair_bk_l_01 [{shapeType=Capsule, length=6, radius=4, shapeName=dyn_hair_bk_l_01_capsule, center=<4.096 0 0>}]
 dyn_hair_bk_l_02 [{shapeType=Capsule, length=6, radius=4, shapeName=dyn_hair_bk_l_02_capsule, center=<4.096 0 0>}]
 dyn_hair_bk_l_03 [{shapeType=Capsule, length=6, radius=4, shapeName=dyn_hair_bk_l_03_capsule, center=<4.096 0 0>}]
 dyn_hair_bk_l_04 [{shapeType=Capsule, length=6, radius=4, shapeName=dyn_hair_bk_l_04_capsule, center=<4.096 0 0>}]
 dyn_hair_bk_l_05 [{shapeType=Capsule, length=6, radius=4, shapeName=dyn_hair_bk_l_05_capsule, center=<4.096 0 0>}]
 dyn_hair_bk_l_06 [{shapeType=Capsule, length=6, radius=4, shapeName=dyn_hair_bk_l_06_capsule, center=<4.096 0 0>}]
 dyn_hair_bk_l_07 [{shapeType=Capsule, length=6, radius=4, shapeName=dyn_hair_bk_l_07_capsule, center=<4.096 0 0>}]
 dyn_hair_bk_l_08 [{shapeType=Capsule, length=11.0275, radius=4.4205, shapeName=dyn_hair_bk_l_08_capsule, center=<4.096 0 0>}]
 dyn_hair_bk_l_09 [{shapeType=Capsule, length=13.2392, radius=5.235, shapeName=dyn_hair_bk_l_09_capsule, center=<4.096 0 0>}]
 dyn_hair_bk_l_13 [{shapeType=Capsule, length=9.4063, radius=8.0318, shapeName=dyn_hair_bk_l_13_capsule, center=<5.2353 0 0>}]
 dyn_hair_bk_l_15 [{shapeType=Capsule, length=7.6865, radius=6.9353, shapeName=dyn_hair_bk_l_15_capsule, center=<9.2377 0 0>, rotation=<rot 90 0 0>}]
 dyn_hair_bk_l_17 [{shapeType=Capsule, length=6, radius=4, shapeName=dyn_hair_bk_l_17_capsule, center=<6.2969 0 0>, rotation=<rot 90 0 0>}]
 dyn_hair_bk_l_18 [{shapeType=Capsule, length=6, radius=4, shapeName=dyn_hair_bk_l_18_capsule, center=<6.2969 0 0>, rotation=<rot 90 0 0>}]
 dyn_hair_bk_l_base [{shapeType=Capsule, length=0, radius=0, shapeName=dyn_hair_bk_l_base_capsule}]
 dyn_hair_bk_mid_01 [{shapeType=Capsule, length=6, radius=4, shapeName=dyn_hair_bk_mid_01_capsule, center=<3.3755 0 0>}]
 dyn_hair_bk_mid_02 [{shapeType=Capsule, length=6, radius=4, shapeName=dyn_hair_bk_mid_02_capsule, center=<3.3755 0 0>}]
 dyn_hair_bk_mid_03 [{shapeType=Capsule, length=6, radius=4, shapeName=dyn_hair_bk_mid_03_capsule, center=<3.3755 0 0>}]
 dyn_hair_bk_mid_04 [{shapeType=Capsule, length=11.0974, radius=4, shapeName=dyn_hair_bk_mid_04_capsule, center=<3.3755 0 0>}]
 dyn_hair_bk_mid_05 [{shapeType=Capsule, length=6, radius=4, shapeName=dyn_hair_bk_mid_05_capsule, center=<3.3755 0 0>}]
 dyn_hair_bk_mid_06 [{shapeType=Capsule, length=6, radius=4, shapeName=dyn_hair_bk_mid_06_capsule, center=<3.3755 0 0>}]
 dyn_hair_bk_mid_07 [{shapeType=Capsule, length=6, radius=4, shapeName=dyn_hair_bk_mid_07_capsule, center=<3.3755 0 0>}]
 dyn_hair_bk_mid_08 [{shapeType=Capsule, length=11.0275, radius=4.4205, shapeName=dyn_hair_bk_mid_08_capsule, center=<3.3755 0 0>}]
 dyn_hair_bk_mid_09 [{shapeType=Capsule, length=13.2392, radius=5.235, shapeName=dyn_hair_bk_mid_09_capsule, center=<3.3755 0 0>}]
 dyn_hair_bk_mid_10 [{shapeType=Capsule, length=9.4063, radius=8.0318, shapeName=dyn_hair_bk_mid_10_capsule, center=<4.5148 0 0>}]
 dyn_hair_bk_mid_11 [{shapeType=Capsule, length=9.4063, radius=8.0318, shapeName=dyn_hair_bk_mid_11_capsule, center=<4.5148 0 0>}]
 dyn_hair_bk_mid_12 [{shapeType=Capsule, length=9.1824, radius=8.1035, shapeName=dyn_hair_bk_mid_12_capsule, center=<5.6083 0 0>, rotation=<rot 90 0 0>}]
 dyn_hair_bk_mid_base [{shapeType=Capsule, length=0, radius=0, shapeName=dyn_hair_bk_mid_base_capsule}]
 dyn_hair_bk_r_01 [{shapeType=Capsule, length=8.4985, radius=4.6899, shapeName=dyn_hair_bk_r_01_capsule, center=<3.5773 0 0>}]
 dyn_hair_bk_r_02 [{shapeType=Capsule, length=5.4078, radius=4.5922, shapeName=dyn_hair_bk_r_02_capsule, center=<3.5773 0 0>}]
 dyn_hair_bk_r_03 [{shapeType=Capsule, length=10.0256, radius=4.1407, shapeName=dyn_hair_bk_r_03_capsule, center=<3.5773 0 0>}]
 dyn_hair_bk_r_04 [{shapeType=Capsule, length=6, radius=4, shapeName=dyn_hair_bk_r_04_capsule, center=<3.5773 0 0>}]
 dyn_hair_bk_r_05 [{shapeType=Capsule, length=6, radius=4, shapeName=dyn_hair_bk_r_05_capsule, center=<3.5773 0 0>}]
 dyn_hair_bk_r_06 [{shapeType=Capsule, length=6, radius=4, shapeName=dyn_hair_bk_r_06_capsule, center=<3.5773 0 0>}]
 dyn_hair_bk_r_07 [{shapeType=Capsule, length=6, radius=4, shapeName=dyn_hair_bk_r_07_capsule, center=<3.5773 0 0>}]
 dyn_hair_bk_r_08 [{shapeType=Capsule, length=6.6917, radius=4.4205, shapeName=dyn_hair_bk_r_08_capsule, center=<3.5773 0 0>}]
 dyn_hair_bk_r_09 [{shapeType=Capsule, length=7.9263, radius=5.235, shapeName=dyn_hair_bk_r_09_capsule, center=<3.5773 0 0>}]
 dyn_hair_bk_r_10 [{shapeType=Capsule, length=6.2578, radius=6.2723, shapeName=dyn_hair_bk_r_10_capsule, center=<4.7166 0 0>}]
 dyn_hair_bk_r_11 [{shapeType=Capsule, length=9.0634, radius=4.8038, shapeName=dyn_hair_bk_r_11_capsule, center=<5.03 0 0>, rotation=<rot 90 0 0>}]
 dyn_hair_bk_r_13 [{shapeType=Capsule, length=15.8649, radius=6.0829, shapeName=dyn_hair_bk_r_13_capsule, center=<10.6229 0 0>, rotation=<rot 90 0 0>}]
 dyn_hair_bk_r_14 [{shapeType=Capsule, length=10.2098, radius=6.5599, shapeName=dyn_hair_bk_r_14_capsule, center=<10.6229 0 0>, rotation=<rot 90 0 0>}]
 dyn_hair_bk_r_15 [{
   shapeType=Capsule
   length=7.7795
   radius=2.9654
   shapeName=dyn_hair_bk_r_15_capsule
   center=<3.9706 -0.0001 1.0619>
   rotation=<rot 75.0277 179.9879 179.99>
  }]
 dyn_hair_bk_r_base [{shapeType=Capsule, length=0, radius=0, shapeName=dyn_hair_bk_r_base_capsule}]
 dyn_hair_bk_side_l_01 [{shapeType=Capsule, length=6.6457, radius=3.3543, shapeName=dyn_hair_bk_side_l_01_capsule, center=<6.3353 0 0>, rotation=<rot 90 0 0>}]
 dyn_hair_bk_side_l_02 [{shapeType=Capsule, length=6.4165, radius=4.1327, shapeName=dyn_hair_bk_side_l_02_capsule, center=<6.3353 0 0>, rotation=<rot 90 0 0>}]
 dyn_hair_bk_side_l_03 [{shapeType=Capsule, length=5.9148, radius=4.0852, shapeName=dyn_hair_bk_side_l_03_capsule, center=<6.3353 0 0>, rotation=<rot 90 0 0>}]
 dyn_hair_bk_side_l_04 [{shapeType=Capsule, length=4.7447, radius=5.2553, shapeName=dyn_hair_bk_side_l_04_capsule, center=<6.3353 -2.0338 0>, rotation=<rot 90 0 0>}]
 dyn_hair_bk_side_l_base [{shapeType=Capsule, length=0, radius=0, shapeName=dyn_hair_bk_side_l_base_capsule}]
 dyn_hair_bk_side_r_01 [{shapeType=Capsule, length=6, radius=4, shapeName=dyn_hair_bk_side_r_01_capsule, center=<6.3353 0 0.6217>, rotation=<rot 90 0 0>}]
 dyn_hair_bk_side_r_02 [{shapeType=Capsule, length=6, radius=4, shapeName=dyn_hair_bk_side_r_02_capsule, center=<6.3353 0 2.4135>, rotation=<rot 90 0 0>}]
 dyn_hair_bk_side_r_03 [{
   shapeType=Capsule
   length=6
   radius=4
   shapeName=dyn_hair_bk_side_r_03_capsule
   center=<4.8673 -0.5614 3.5232>
   rotation=<rot 70.3554 173.4204 179.9998>
  }]
 dyn_hair_bk_side_r_base [{shapeType=Capsule, length=0, radius=0, shapeName=dyn_hair_bk_side_r_base_capsule}]
 dyn_hair_frnt_r_01 [{shapeType=Capsule, length=6.6274, radius=3.3726, shapeName=dyn_hair_frnt_r_01_capsule, center=<6.3353 0 0>, rotation=<rot 90 0 0>}]
 dyn_hair_frnt_r_02 [{shapeType=Capsule, length=6.3858, radius=3.6142, shapeName=dyn_hair_frnt_r_02_capsule, center=<6.3353 0 0>, rotation=<rot 90 0 0>}]
 dyn_hair_frnt_r_03 [{shapeType=Capsule, length=6, radius=4, shapeName=dyn_hair_frnt_r_03_capsule, center=<6.3353 0 0>, rotation=<rot 90 0 0>}]
 dyn_hair_frnt_r_04 [{
   shapeType=Capsule
   length=6.9905
   radius=3.0095
   shapeName=dyn_hair_frnt_r_04_capsule
   center=<4.823 0.7158 1.4013>
   rotation=<rot 73.9656 -171.5588 179.9997>
  }]
 dyn_hair_frnt_r_05 [{
   shapeType=Capsule
   length=5.7974
   radius=2.6026
   shapeName=dyn_hair_frnt_r_05_capsule
   center=<4.0572 -1.1551 0>
   rotation=<rot 90 -21.5699 0>
  }]
 dyn_hair_frnt_r_06 [{
   shapeType=Capsule
   length=10.4957
   radius=3.3021
   shapeName=dyn_hair_frnt_r_06_capsule
   center=<8.3228 0 2.415>
   rotation=<rot 73.0011 179.9997 179.9999>
  }]
 dyn_hair_frnt_r_base [{shapeType=Capsule, length=0, radius=0, shapeName=dyn_hair_frnt_r_base_capsule}]
 head [{shapeType=Capsule, length=3.9006, radius=7.1495, shapeName=head_capsule, rotation=<rot 90 11.2244 0>}]
 neck_01 [{shapeType=Capsule, length=9.2766, radius=4.688, shapeName=neck_01_capsule, rotation=<rot 90 0 0>}]
 neck_02 [{shapeType=Capsule, length=24.8523, radius=4.1776, shapeName=neck_02_capsule, center=<10.9768 0 0>, rotation=<rot 90 0 0>}]
 pelvis [{shapeType=Capsule, length=15.3027, radius=11.2158, shapeName=pelvis_capsule, center=<-5.6338 -0.3505 0>}]
 spine_01 [{shapeType=Capsule, length=15.4397, radius=9.9257, shapeName=spine_01_capsule, center=<0 1.8698 0>}]
 spine_02 [{shapeType=Capsule, length=13.1451, radius=8.5354, shapeName=spine_02_capsule, center=<0 1.4394 0>}]
 spine_03 [{shapeType=Capsule, length=16.8967, radius=10.5218, shapeName=spine_03_capsule, center=<0 1.5858 0>}]
 spine_04 [{shapeType=Capsule, length=17.3755, radius=12.5502, shapeName=spine_04_capsule, center=<0 3.4628 0>}]
 spine_05 [{shapeType=Capsule, length=11.5646, radius=10.3432, shapeName=spine_05_capsule, center=<0 1.8196 0>}]
 upperarm_l [
  {shapeType=Sphere, radius=5, shapeName=upperarm_l_sphere}
  {shapeType=Capsule, length=25.1974, radius=5.1848, shapeName=upperarm_l_capsule, center=<13.6666 0 0>, rotation=<rot 90 0 0>}
 ]
 upperarm_r [
  {shapeType=Sphere, radius=6.268, shapeName=upperarm_r_sphere}
  {shapeType=Capsule, length=25.1974, radius=5.1848, shapeName=upperarm_r_capsule, center=<-13.6666 0 0>, rotation=<rot 90 180 0>}
 ]
}
constraints [
 {
  bone1Name=spine_01
  bone2Name=pelvis
  swing1LimitDegrees=45
  swing1Motion=Free
  swing2LimitDegrees=45
  swing2Motion=Free
  twistLimitDegrees=45
  twistMotion=Free
 }
 {
  bone1Name=spine_02
  bone2Name=spine_01
  swing1LimitDegrees=45
  swing1Motion=Free
  swing2LimitDegrees=45
  swing2Motion=Free
  twistLimitDegrees=45
  twistMotion=Free
 }
 {
  bone1Name=spine_03
  bone2Name=spine_02
  swing1LimitDegrees=45
  swing1Motion=Free
  swing2LimitDegrees=45
  swing2Motion=Free
  twistLimitDegrees=45
  twistMotion=Free
 }
 {
  bone1Name=spine_04
  bone2Name=spine_03
  swing1LimitDegrees=45
  swing1Motion=Free
  swing2LimitDegrees=45
  swing2Motion=Free
  twistLimitDegrees=45
  twistMotion=Free
 }
 {
  bone1Name=spine_05
  bone2Name=spine_04
  swing1LimitDegrees=45
  swing1Motion=Free
  swing2LimitDegrees=45
  swing2Motion=Free
  twistLimitDegrees=45
  twistMotion=Free
 }
 {
  bone1Name=clavicle_l
  bone2Name=spine_05
  swing1LimitDegrees=45
  swing1Motion=Free
  swing2LimitDegrees=45
  swing2Motion=Free
  twistLimitDegrees=45
  twistMotion=Free
 }
 {
  bone1Name=upperarm_l
  bone2Name=clavicle_l
  swing1LimitDegrees=45
  swing1Motion=Free
  swing2LimitDegrees=45
  swing2Motion=Free
  twistLimitDegrees=45
  twistMotion=Free
 }
 {
  bone1Name=clavicle_r
  bone2Name=spine_05
  swing1LimitDegrees=45
  swing1Motion=Free
  swing2LimitDegrees=45
  swing2Motion=Free
  twistLimitDegrees=45
  twistMotion=Free
 }
 {
  bone1Name=upperarm_r
  bone2Name=clavicle_r
  swing1LimitDegrees=45
  swing1Motion=Free
  swing2LimitDegrees=45
  swing2Motion=Free
  twistLimitDegrees=45
  twistMotion=Free
 }
 {
  bone1Name=neck_01
  bone2Name=spine_05
  swing1LimitDegrees=45
  swing1Motion=Free
  swing2LimitDegrees=45
  swing2Motion=Free
  twistLimitDegrees=45
  twistMotion=Free
 }
 {
  bone1Name=neck_02
  bone2Name=neck_01
  swing1LimitDegrees=45
  swing1Motion=Free
  swing2LimitDegrees=45
  swing2Motion=Free
  twistLimitDegrees=45
  twistMotion=Free
 }
 {
  bone1Name=head
  bone2Name=neck_02
  swing1LimitDegrees=45
  swing1Motion=Free
  swing2LimitDegrees=45
  swing2Motion=Free
  twistLimitDegrees=45
  twistMotion=Free
 }
 {
  bone1Name=dyn_hair_bk_l_base
  bone2Name=head
  swing1LimitDegrees=45
  swing1Motion=Free
  swing2LimitDegrees=45
  swing2Motion=Free
  twistLimitDegrees=45
  twistMotion=Free
 }
 {bone1Name=dyn_hair_bk_l_01, bone2Name=dyn_hair_bk_l_base, swing1LimitDegrees=10, swing2LimitDegrees=0, twistLimitDegrees=0}
 {bone1Name=dyn_hair_bk_l_02, bone2Name=dyn_hair_bk_l_01, swing1LimitDegrees=35, swing2LimitDegrees=0, twistLimitDegrees=0}
 {bone1Name=dyn_hair_bk_l_03, bone2Name=dyn_hair_bk_l_02, swing1LimitDegrees=20, swing2LimitDegrees=0, twistLimitDegrees=5}
 {bone1Name=dyn_hair_bk_l_04, bone2Name=dyn_hair_bk_l_03, swing1LimitDegrees=20, swing2LimitDegrees=0, twistLimitDegrees=5}
 {bone1Name=dyn_hair_bk_l_05, bone2Name=dyn_hair_bk_l_04, swing1LimitDegrees=20, swing2LimitDegrees=0, twistLimitDegrees=5}
 {bone1Name=dyn_hair_bk_l_06, bone2Name=dyn_hair_bk_l_05, swing1LimitDegrees=20, swing2LimitDegrees=0, twistLimitDegrees=5}
 {bone1Name=dyn_hair_bk_l_07, bone2Name=dyn_hair_bk_l_06, swing1LimitDegrees=20, swing2LimitDegrees=0, twistLimitDegrees=5}
 {bone1Name=dyn_hair_bk_l_08, bone2Name=dyn_hair_bk_l_07, swing1LimitDegrees=20, swing2LimitDegrees=5, twistLimitDegrees=5}
 {bone1Name=dyn_hair_bk_l_09, bone2Name=dyn_hair_bk_l_08, swing1LimitDegrees=20, swing2LimitDegrees=5, twistLimitDegrees=5}
 {bone1Name=dyn_hair_bk_l_13, bone2Name=dyn_hair_bk_l_09, swing1LimitDegrees=20, swing2LimitDegrees=5, twistLimitDegrees=5}
 {bone1Name=dyn_hair_bk_l_15, bone2Name=dyn_hair_bk_l_06, swing1LimitDegrees=20, swing2LimitDegrees=0, twistLimitDegrees=5}
 {bone1Name=dyn_hair_bk_l_17, bone2Name=dyn_hair_bk_l_04, swing1LimitDegrees=20, swing2LimitDegrees=0, twistLimitDegrees=5}
 {bone1Name=dyn_hair_bk_l_18, bone2Name=dyn_hair_bk_l_17, swing1LimitDegrees=20, swing2LimitDegrees=5, twistLimitDegrees=10}
 {
  bone1Name=dyn_hair_bk_mid_base
  bone2Name=head
  swing1LimitDegrees=45
  swing1Motion=Free
  swing2LimitDegrees=45
  swing2Motion=Free
  twistLimitDegrees=45
  twistMotion=Free
 }
 {bone1Name=dyn_hair_bk_mid_01, bone2Name=dyn_hair_bk_mid_base, swing1LimitDegrees=10, swing2LimitDegrees=0, twistLimitDegrees=0}
 {bone1Name=dyn_hair_bk_mid_02, bone2Name=dyn_hair_bk_mid_01, swing1LimitDegrees=35, swing2LimitDegrees=0, twistLimitDegrees=0}
 {bone1Name=dyn_hair_bk_mid_03, bone2Name=dyn_hair_bk_mid_02, swing1LimitDegrees=20, swing2LimitDegrees=0, twistLimitDegrees=5}
 {bone1Name=dyn_hair_bk_mid_04, bone2Name=dyn_hair_bk_mid_03, swing1LimitDegrees=20, swing2LimitDegrees=0, twistLimitDegrees=5}
 {bone1Name=dyn_hair_bk_mid_05, bone2Name=dyn_hair_bk_mid_04, swing1LimitDegrees=20, swing2LimitDegrees=0, twistLimitDegrees=5}
 {bone1Name=dyn_hair_bk_mid_06, bone2Name=dyn_hair_bk_mid_05, swing1LimitDegrees=20, swing2LimitDegrees=0, twistLimitDegrees=5}
 {bone1Name=dyn_hair_bk_mid_07, bone2Name=dyn_hair_bk_mid_06, swing1LimitDegrees=20, swing2LimitDegrees=0, twistLimitDegrees=5}
 {bone1Name=dyn_hair_bk_mid_08, bone2Name=dyn_hair_bk_mid_07, swing1LimitDegrees=20, swing2LimitDegrees=5, twistLimitDegrees=5}
 {bone1Name=dyn_hair_bk_mid_09, bone2Name=dyn_hair_bk_mid_08, swing1LimitDegrees=20, swing2LimitDegrees=5, twistLimitDegrees=5}
 {bone1Name=dyn_hair_bk_mid_10, bone2Name=dyn_hair_bk_mid_09, swing1LimitDegrees=20, swing2LimitDegrees=5, twistLimitDegrees=5}
 {bone1Name=dyn_hair_bk_mid_11, bone2Name=dyn_hair_bk_mid_10, swing1LimitDegrees=20, swing2LimitDegrees=5, twistLimitDegrees=5}
 {bone1Name=dyn_hair_bk_mid_12, bone2Name=dyn_hair_bk_mid_07, swing1LimitDegrees=20, swing2LimitDegrees=5, twistLimitDegrees=5}
 {
  bone1Name=dyn_hair_bk_r_base
  bone2Name=head
  swing1LimitDegrees=45
  swing1Motion=Free
  swing2LimitDegrees=45
  swing2Motion=Free
  twistLimitDegrees=45
  twistMotion=Free
 }
 {bone1Name=dyn_hair_bk_r_01, bone2Name=dyn_hair_bk_r_base, swing1LimitDegrees=30, swing2LimitDegrees=0, twistLimitDegrees=20}
 {bone1Name=dyn_hair_bk_r_02, bone2Name=dyn_hair_bk_r_01, swing1LimitDegrees=35, swing2LimitDegrees=0, twistLimitDegrees=0}
 {bone1Name=dyn_hair_bk_r_03, bone2Name=dyn_hair_bk_r_02, swing1LimitDegrees=35, swing2LimitDegrees=0, twistLimitDegrees=5}
 {bone1Name=dyn_hair_bk_r_04, bone2Name=dyn_hair_bk_r_03, swing1LimitDegrees=20, swing2LimitDegrees=0, twistLimitDegrees=5}
 {bone1Name=dyn_hair_bk_r_05, bone2Name=dyn_hair_bk_r_04, swing1LimitDegrees=20, swing2LimitDegrees=0, twistLimitDegrees=5}
 {bone1Name=dyn_hair_bk_r_06, bone2Name=dyn_hair_bk_r_05, swing1LimitDegrees=20, swing2LimitDegrees=0, twistLimitDegrees=5}
 {bone1Name=dyn_hair_bk_r_07, bone2Name=dyn_hair_bk_r_06, swing1LimitDegrees=20, swing2LimitDegrees=0, twistLimitDegrees=5}
 {bone1Name=dyn_hair_bk_r_08, bone2Name=dyn_hair_bk_r_07, swing1LimitDegrees=20, swing2LimitDegrees=5, twistLimitDegrees=5}
 {bone1Name=dyn_hair_bk_r_09, bone2Name=dyn_hair_bk_r_08, swing1LimitDegrees=20, swing2LimitDegrees=5, twistLimitDegrees=5}
 {bone1Name=dyn_hair_bk_r_10, bone2Name=dyn_hair_bk_r_09, swing1LimitDegrees=20, swing2LimitDegrees=5, twistLimitDegrees=5}
 {bone1Name=dyn_hair_bk_r_11, bone2Name=dyn_hair_bk_r_09, swing1LimitDegrees=20, swing2LimitDegrees=5, twistLimitDegrees=5}
 {bone1Name=dyn_hair_bk_r_13, bone2Name=dyn_hair_bk_r_05, swing1LimitDegrees=20, swing2LimitDegrees=0, twistLimitDegrees=5}
 {bone1Name=dyn_hair_bk_r_14, bone2Name=dyn_hair_bk_r_13, swing1LimitDegrees=20, swing2LimitDegrees=5, twistLimitDegrees=10}
 {bone1Name=dyn_hair_bk_r_15, bone2Name=dyn_hair_bk_r_03, swing1LimitDegrees=20, swing2LimitDegrees=45, twistLimitDegrees=20}
 {
  bone1Name=dyn_hair_bk_side_l_base
  bone2Name=head
  swing1LimitDegrees=45
  swing1Motion=Free
  swing2LimitDegrees=45
  swing2Motion=Free
  twistLimitDegrees=45
  twistMotion=Free
 }
 {bone1Name=dyn_hair_bk_side_l_01, bone2Name=dyn_hair_bk_side_l_base, swing1LimitDegrees=20, swing2LimitDegrees=25, twistLimitDegrees=10}
 {bone1Name=dyn_hair_bk_side_l_02, bone2Name=dyn_hair_bk_side_l_01, swing1LimitDegrees=30, swing2LimitDegrees=30, twistLimitDegrees=10}
 {bone1Name=dyn_hair_bk_side_l_03, bone2Name=dyn_hair_bk_side_l_02, swing1LimitDegrees=30, swing2LimitDegrees=30, twistLimitDegrees=10}
 {bone1Name=dyn_hair_bk_side_l_04, bone2Name=dyn_hair_bk_side_l_01, swing1LimitDegrees=20, swing2LimitDegrees=5, twistLimitDegrees=10}
 {
  bone1Name=dyn_hair_bk_side_r_base
  bone2Name=head
  swing1LimitDegrees=45
  swing1Motion=Free
  swing2LimitDegrees=45
  swing2Motion=Free
  twistLimitDegrees=45
  twistMotion=Free
 }
 {bone1Name=dyn_hair_bk_side_r_01, bone2Name=dyn_hair_bk_side_r_base, swing1LimitDegrees=20, swing2LimitDegrees=20, twistLimitDegrees=10}
 {bone1Name=dyn_hair_bk_side_r_02, bone2Name=dyn_hair_bk_side_r_01, swing1LimitDegrees=20, swing2LimitDegrees=20, twistLimitDegrees=10}
 {bone1Name=dyn_hair_bk_side_r_03, bone2Name=dyn_hair_bk_side_r_02, swing1LimitDegrees=20, swing2LimitDegrees=20, twistLimitDegrees=10}
 {
  bone1Name=dyn_hair_frnt_r_base
  bone2Name=head
  swing1LimitDegrees=45
  swing1Motion=Free
  swing2LimitDegrees=45
  swing2Motion=Free
  twistLimitDegrees=45
  twistMotion=Free
 }
 {bone1Name=dyn_hair_frnt_r_01, bone2Name=dyn_hair_frnt_r_base, swing1LimitDegrees=20, swing2LimitDegrees=20, twistLimitDegrees=10}
 {bone1Name=dyn_hair_frnt_r_02, bone2Name=dyn_hair_frnt_r_01, swing1LimitDegrees=20, swing2LimitDegrees=20, twistLimitDegrees=10}
 {bone1Name=dyn_hair_frnt_r_03, bone2Name=dyn_hair_frnt_r_02, swing1LimitDegrees=20, swing2LimitDegrees=20, twistLimitDegrees=10}
 {bone1Name=dyn_hair_frnt_r_05, bone2Name=dyn_hair_frnt_r_03, swing1LimitDegrees=20, swing2LimitDegrees=20, twistLimitDegrees=10}
 {bone1Name=dyn_hair_frnt_r_04, bone2Name=dyn_hair_frnt_r_02, swing1LimitDegrees=20, swing2LimitDegrees=20, twistLimitDegrees=10}
 {bone1Name=dyn_hair_frnt_r_06, bone2Name=dyn_hair_frnt_r_02, swing1LimitDegrees=20, swing2LimitDegrees=20, twistLimitDegrees=10}
]
